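# Niagara system for the Target Practice skin that places blinking blue, pink and yellow lights at front and back offsets on the character.
# Six emitters each burst one particle with sprite and light renderers, the front lights having brighter colors and larger radii.
# User timing and reactive parameters, named as Blueprint-fed game time, copy into each emitter alongside a per-color blink order value.
# A scratch multiplier scales each light's radius every update, apparently driving the blink.
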NiagaraSystem /BRCosmetics/Characters/Player/Male/Medium/Bodies/M_MED_Target_Practice/Effects/NS_TargetPractice_SkinLight
dependencies {usedDataInterfaces [], usedDynamicInputs [], usedModules [&SystemState]}
emitters {
 BlueLightBack {
  emitterUpdateScript [
   &EmitterState(Life Cycle Mode=::NewEnumerator1, Loop Duration=1, UseLoopDelay=false)
   &SpawnBurst_Instantaneous(Spawn Count=1, Use Spawn Probability=false)
   /SetVariables(
    Emitter.BlinkOrderVertColor=0.753
    Emitter.TP_GameTimeFromBP:NiagaraFloat=$User.TP_GameTimeFromBP
    Emitter.TP_LightsDefaultSpeed:NiagaraFloat=$User.TP_LightsDefaultSpeed
    Emitter.TP_LightsReactiveSpeed:NiagaraFloat=$User.TP_LightsReactiveSpeed
    Emitter.TP_ReactiveBlinkDuration:NiagaraFloat=$User.TP_ReactiveBlinkDuration
    Emitter.TP_ReactiveGlowDuration:NiagaraFloat=$User.TP_ReactiveGlowDuration
    Emitter.TP_TimeOffset:NiagaraFloat=$User.TP_TimeOffset
    Emitter.TP_UseVertexColorChannelForTimeOffset:NiagaraFloat=$User.TP_UseVertexColorChannelForTimeOFfset
    Emitter.TP_ValueToHideonReactive:NiagaraFloat=$User.TP_ValueToHideonReactive
   ) {isSetParameters=true, module=SetVariables_#1}
  ]
  particleSpawnScript [
   /Niagara/Modules/Spawn/Initialization/InitializeParticle(
    Color=<rgba 0.0515 1.3069 1.5 0>
    Module.Write Color=true
    Module.Write Lifetime=true
    Module.Write Mass=true
    Module.Write Material Random=true
    Module.Write Position=false
    Module.Write Scale=false
    Module.Write SpriteRotation=false
    Module.Write SpriteSize=true
    Sprite Size=<5 5>
   )
   /SetVariables(Particles.LightRadius=12) {isSetParameters=true, module=SetVariables_#2}
   &SystemLocation(MaskSpawnGroup=false, Offset=<-7.3102 -16.8545 0>)
  ]
  particleUpdateScript [
   &ParticleState()
   &SolveForcesAndVelocity(Clamp Velocity=false, Limit Acceleration=false)
   /SetVariables(Particles.NewLightRadius:NiagaraFloat=&Multiply_Float(
     A=$Particles.LightRadius
     B=&self:MultiplyLightRadiusScratch()
    )) {isSetParameters=true, module=SetVariables_#3}
  ]
  rendererClasses [&NiagaraSpriteRendererProperties, &NiagaraLightRendererProperties]
  renderers [
   {rendererClass=&NiagaraSpriteRendererProperties}
   {rendererClass=&NiagaraLightRendererProperties, rendererIndex=1}
  ]
 }
 BlueLightFront {
  emitterUpdateScript [
   &EmitterState(Life Cycle Mode=::NewEnumerator1, Loop Duration=1, UseLoopDelay=false)
   &SpawnBurst_Instantaneous(Spawn Count=1, Use Spawn Probability=false)
   /SetVariables(
    Emitter.BlinkOrderVertColor=0.753
    Emitter.TP_GameTimeFromBP:NiagaraFloat=$User.TP_GameTimeFromBP
    Emitter.TP_LightsDefaultSpeed:NiagaraFloat=$User.TP_LightsDefaultSpeed
    Emitter.TP_LightsReactiveSpeed:NiagaraFloat=$User.TP_LightsReactiveSpeed
    Emitter.TP_ReactiveBlinkDuration:NiagaraFloat=$User.TP_ReactiveBlinkDuration
    Emitter.TP_ReactiveGlowDuration:NiagaraFloat=$User.TP_ReactiveGlowDuration
    Emitter.TP_TimeOffset:NiagaraFloat=$User.TP_TimeOffset
    Emitter.TP_UseVertexColorChannelForTimeOffset:NiagaraFloat=$User.TP_UseVertexColorChannelForTimeOFfset
    Emitter.TP_ValueToHideonReactive:NiagaraFloat=$User.TP_ValueToHideonReactive
   ) {isSetParameters=true, module=SetVariables_#1}
  ]
  particleSpawnScript [
   &InitializeParticle(
    Color=<rgba 0.1373 3.485 4 0>
    Module.Write Color=true
    Module.Write Lifetime=true
    Module.Write Mass=true
    Module.Write Material Random=true
    Module.Write Position=false
    Module.Write Scale=false
    Module.Write SpriteRotation=false
    Module.Write SpriteSize=true
    Sprite Size=<5 5>
   )
   /SetVariables(Particles.LightRadius=23) {isSetParameters=true, module=SetVariables_#2}
   &SystemLocation(MaskSpawnGroup=false, Offset=<-5.156 18.4156 0>)
  ]
  particleUpdateScript [
   &ParticleState()
   &SolveForcesAndVelocity(Clamp Velocity=false, Limit Acceleration=false)
   /SetVariables(Particles.NewLightRadius:NiagaraFloat=&Multiply_Float(
     A=$Particles.LightRadius
     B=&self:MultiplyLightRadiusScratch()
    )) {isSetParameters=true, module=SetVariables_#3}
  ]
  rendererClasses [&NiagaraSpriteRendererProperties, &NiagaraLightRendererProperties]
  renderers [
   {rendererClass=&NiagaraSpriteRendererProperties}
   {rendererClass=&NiagaraLightRendererProperties, rendererIndex=1}
  ]
 }
 PinkLightBack {
  emitterUpdateScript [
   &EmitterState(Life Cycle Mode=::NewEnumerator1, Loop Duration=1, UseLoopDelay=false)
   &SpawnBurst_Instantaneous(Spawn Count=1, Use Spawn Probability=false)
   /SetVariables(
    Emitter.BlinkOrderVertColor=0.502
    Emitter.TP_GameTimeFromBP:NiagaraFloat=$User.TP_GameTimeFromBP
    Emitter.TP_LightsDefaultSpeed:NiagaraFloat=$User.TP_LightsDefaultSpeed
    Emitter.TP_LightsReactiveSpeed:NiagaraFloat=$User.TP_LightsReactiveSpeed
    Emitter.TP_ReactiveBlinkDuration:NiagaraFloat=$User.TP_ReactiveBlinkDuration
    Emitter.TP_ReactiveGlowDuration:NiagaraFloat=$User.TP_ReactiveGlowDuration
    Emitter.TP_TimeOffset:NiagaraFloat=$User.TP_TimeOffset
    Emitter.TP_UseVertexColorChannelForTimeOffset:NiagaraFloat=$User.TP_UseVertexColorChannelForTimeOFfset
    Emitter.TP_ValueToHideonReactive:NiagaraFloat=$User.TP_ValueToHideonReactive
    Emitter.TimingOffsetVertColor=0.502
   ) {isSetParameters=true, module=SetVariables_#1}
  ]
  particleSpawnScript [
   &InitializeParticle(
    Color=<rgba 1.5 0.0515 0.6194 0>
    Module.Write Color=true
    Module.Write Lifetime=true
    Module.Write Mass=true
    Module.Write Material Random=true
    Module.Write Position=false
    Module.Write Scale=false
    Module.Write SpriteRotation=false
    Module.Write SpriteSize=true
    Sprite Size=<5 5>
   )
   /SetVariables(Particles.LightRadius=12) {isSetParameters=true, module=SetVariables_#2}
   &SystemLocation(MaskSpawnGroup=false, Offset=<-7.3102 -16.8545 0>)
  ]
  particleUpdateScript [
   &ParticleState()
   &SolveForcesAndVelocity(Clamp Velocity=false, Limit Acceleration=false)
   /SetVariables(Particles.NewLightRadius:NiagaraFloat=&Multiply_Float(
     A=$Particles.LightRadius
     B=&self:MultiplyLightRadiusScratch()
    )) {isSetParameters=true, module=SetVariables_#3}
  ]
  rendererClasses [&NiagaraSpriteRendererProperties, &NiagaraLightRendererProperties]
  renderers [
   {rendererClass=&NiagaraSpriteRendererProperties}
   {rendererClass=&NiagaraLightRendererProperties, rendererIndex=1}
  ]
 }
 PinkLightFront {
  emitterUpdateScript [
   &EmitterState(Life Cycle Mode=::NewEnumerator1, Loop Duration=1, UseLoopDelay=false)
   &SpawnBurst_Instantaneous(Spawn Count=1, Use Spawn Probability=false)
   /SetVariables(
    Emitter.BlinkOrderVertColor=0.502
    Emitter.TP_GameTimeFromBP:NiagaraFloat=$User.TP_GameTimeFromBP
    Emitter.TP_LightsDefaultSpeed:NiagaraFloat=$User.TP_LightsDefaultSpeed
    Emitter.TP_LightsReactiveSpeed:NiagaraFloat=$User.TP_LightsReactiveSpeed
    Emitter.TP_ReactiveBlinkDuration:NiagaraFloat=$User.TP_ReactiveBlinkDuration
    Emitter.TP_ReactiveGlowDuration:NiagaraFloat=$User.TP_ReactiveGlowDuration
    Emitter.TP_TimeOffset:NiagaraFloat=$User.TP_TimeOffset
    Emitter.TP_UseVertexColorChannelForTimeOffset:NiagaraFloat=$User.TP_UseVertexColorChannelForTimeOFfset
    Emitter.TP_ValueToHideonReactive:NiagaraFloat=$User.TP_ValueToHideonReactive
    Emitter.TimingOffsetVertColor=0.502
   ) {isSetParameters=true, module=SetVariables_#1}
  ]
  particleSpawnScript [
   &InitializeParticle(
    Color=<rgba 3 0.103 1.2388 0>
    Module.Write Color=true
    Module.Write Lifetime=true
    Module.Write Mass=true
    Module.Write Material Random=true
    Module.Write Position=false
    Module.Write Scale=false
    Module.Write SpriteRotation=false
    Module.Write SpriteSize=true
    Sprite Size=<5 5>
   )
   /SetVariables(Particles.LightRadius=23) {isSetParameters=true, module=SetVariables_#2}
   &SystemLocation(MaskSpawnGroup=false, Offset=<-5.156 18.4156 0>)
  ]
  particleUpdateScript [
   &ParticleState()
   &SolveForcesAndVelocity(Clamp Velocity=false, Limit Acceleration=false)
   /SetVariables(Particles.NewLightRadius:NiagaraFloat=&Multiply_Float(
     A=$Particles.LightRadius
     B=&self:MultiplyLightRadiusScratch()
    )) {isSetParameters=true, module=SetVariables_#3}
  ]
  rendererClasses [&NiagaraSpriteRendererProperties, &NiagaraLightRendererProperties]
  renderers [
   {rendererClass=&NiagaraSpriteRendererProperties}
   {rendererClass=&NiagaraLightRendererProperties, rendererIndex=1}
  ]
 }
 YellowLightBack {
  emitterUpdateScript [
   &EmitterState(Life Cycle Mode=::NewEnumerator1, Loop Duration=1, UseLoopDelay=false)
   &SpawnBurst_Instantaneous(Spawn Count=1, Use Spawn Probability=false)
   /SetVariables(
    Emitter.BlinkOrderVertColor=0.251
    Emitter.TP_GameTimeFromBP:NiagaraFloat=$User.TP_GameTimeFromBP
    Emitter.TP_LightsDefaultSpeed:NiagaraFloat=$User.TP_LightsDefaultSpeed
    Emitter.TP_LightsReactiveSpeed:NiagaraFloat=$User.TP_LightsReactiveSpeed
    Emitter.TP_ReactiveBlinkDuration:NiagaraFloat=$User.TP_ReactiveBlinkDuration
    Emitter.TP_ReactiveGlowDuration:NiagaraFloat=$User.TP_ReactiveGlowDuration
    Emitter.TP_TimeOffset:NiagaraFloat=$User.TP_TimeOffset
    Emitter.TP_UseVertexColorChannelForTimeOffset:NiagaraFloat=$User.TP_UseVertexColorChannelForTimeOFfset
    Emitter.TP_ValueToHideonReactive:NiagaraFloat=$User.TP_ValueToHideonReactive
   ) {isSetParameters=true, module=SetVariables_#1}
  ]
  particleSpawnScript [
   &InitializeParticle(
    Color=<rgba 1 0.4045 0.0343 0>
    Module.Write Color=true
    Module.Write Lifetime=true
    Module.Write Mass=true
    Module.Write Material Random=true
    Module.Write Position=false
    Module.Write Scale=false
    Module.Write SpriteRotation=false
    Module.Write SpriteSize=true
    Sprite Size=<5 5>
   )
   /SetVariables(Particles.LightRadius=12) {isSetParameters=true, module=SetVariables_#2}
   &SystemLocation(MaskSpawnGroup=false, Offset=<-7.3102 -16.8545 0>)
  ]
  particleUpdateScript [
   &ParticleState()
   &SolveForcesAndVelocity(Clamp Velocity=false, Limit Acceleration=false)
   /SetVariables(Particles.NewLightRadius:NiagaraFloat=&Multiply_Float(
     A=$Particles.LightRadius
     B=&self:MultiplyLightRadiusScratch()
    )) {isSetParameters=true, module=SetVariables_#3}
  ]
  rendererClasses [&NiagaraSpriteRendererProperties, &NiagaraLightRendererProperties]
  renderers [
   {rendererClass=&NiagaraSpriteRendererProperties}
   {rendererClass=&NiagaraLightRendererProperties, rendererIndex=1}
  ]
 }
 YellowLightFront {
  emitterUpdateScript [
   &EmitterState(Life Cycle Mode=::NewEnumerator1, Loop Duration=1, UseLoopDelay=false)
   &SpawnBurst_Instantaneous(Spawn Count=1, Use Spawn Probability=false)
   /SetVariables(
    Emitter.BlinkOrderVertColor=0.251
    Emitter.TP_GameTimeFromBP:NiagaraFloat=$User.TP_GameTimeFromBP
    Emitter.TP_LightsDefaultSpeed:NiagaraFloat=$User.TP_LightsDefaultSpeed
    Emitter.TP_LightsReactiveSpeed:NiagaraFloat=$User.TP_LightsReactiveSpeed
    Emitter.TP_ReactiveBlinkDuration:NiagaraFloat=$User.TP_ReactiveBlinkDuration
    Emitter.TP_ReactiveGlowDuration:NiagaraFloat=$User.TP_ReactiveGlowDuration
    Emitter.TP_TimeOffset:NiagaraFloat=$User.TP_TimeOffset
    Emitter.TP_UseVertexColorChannelForTimeOffset:NiagaraFloat=$User.TP_UseVertexColorChannelForTimeOFfset
    Emitter.TP_ValueToHideonReactive:NiagaraFloat=$User.TP_ValueToHideonReactive
   ) {isSetParameters=true, module=SetVariables_#1}
  ]
  particleSpawnScript [
   &InitializeParticle(
    Color=<rgba 2 0.809 0.0687 0>
    Module.Write Color=true
    Module.Write Lifetime=true
    Module.Write Mass=true
    Module.Write Material Random=true
    Module.Write Position=false
    Module.Write Scale=false
    Module.Write SpriteRotation=false
    Module.Write SpriteSize=true
    Sprite Size=<5 5>
   )
   /SetVariables(Particles.LightRadius=23) {isSetParameters=true, module=SetVariables_#2}
   &SystemLocation(MaskSpawnGroup=false, Offset=<-5.156 18.4156 0>)
  ]
  particleUpdateScript [
   &ParticleState()
   &SolveForcesAndVelocity(Clamp Velocity=false, Limit Acceleration=false)
   /SetVariables(Particles.NewLightRadius:NiagaraFloat=&Multiply_Float(
     A=$Particles.LightRadius
     B=&self:MultiplyLightRadiusScratch()
    )) {isSetParameters=true, module=SetVariables_#3}
  ]
  rendererClasses [&NiagaraSpriteRendererProperties, &NiagaraLightRendererProperties]
  renderers [
   {rendererClass=&NiagaraSpriteRendererProperties}
   {rendererClass=&NiagaraLightRendererProperties, rendererIndex=1}
  ]
 }
}
summary {
 systemName=NS_TargetPractice_SkinLight
 emitters [
  {
   emitterName=PinkLightFront
   rendererClasses [&NiagaraSpriteRendererProperties, &NiagaraLightRendererProperties]
  }
  {
   emitterName=BlueLightFront
   rendererClasses [&NiagaraSpriteRendererProperties, &NiagaraLightRendererProperties]
  }
  {
   emitterName=YellowLightFront
   rendererClasses [&NiagaraSpriteRendererProperties, &NiagaraLightRendererProperties]
  }
  {emitterName=PinkLightBack, rendererClasses [&NiagaraSpriteRendererProperties, &NiagaraLightRendererProperties]}
  {emitterName=BlueLightBack, rendererClasses [&NiagaraSpriteRendererProperties, &NiagaraLightRendererProperties]}
  {
   emitterName=YellowLightBack
   rendererClasses [&NiagaraSpriteRendererProperties, &NiagaraLightRendererProperties]
  }
 ]
 userVariables [
  User.TP_GameTimeFromBP {type=NiagaraFloat, defaultValue=0, description=""}
  User.TP_LightsDefaultSpeed {type=NiagaraFloat, defaultValue=0, description=""}
  User.TP_LightsReactiveSpeed {type=NiagaraFloat, defaultValue=0, description=""}
  User.TP_ReactiveBlinkDuration {type=NiagaraFloat, defaultValue=0, description=""}
  User.TP_ReactiveGlowDuration {type=NiagaraFloat, defaultValue=0, description=""}
  User.TP_TimeOffset {type=NiagaraFloat, defaultValue=0, description=""}
  User.TP_UseVertexColorChannelForTimeOFfset {type=NiagaraFloat, defaultValue=0, description=""}
  User.TP_ValueToHideonReactive {type=NiagaraFloat, defaultValue=0, description=""}
 ]
}
